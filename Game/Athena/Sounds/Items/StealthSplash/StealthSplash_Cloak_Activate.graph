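# Sound cue for StealthSplash cloak activation, playing a single wave after a short fixed delay in the gadget/item sound class.
# IsLocal and IsEnemy branches pick the playback volume: local loudest, enemy slightly lower, non-enemy lowest, with unset IsEnemy using the enemy level.
# All three paths use the cloak on/off attenuation and share one downstream chain of mixer, neutral modulator, delay and wave player.
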SoundCue /Game/Athena/Sounds/Items/StealthSplash/StealthSplash_Cloak_Activate
root=Branch_0
nodes [
 {id=Attenuation_0, children [Mixer_0], properties {attenuationSettings=&/Att_StealthSplash_Cloak_OnOff}}
 {id=Attenuation_1, children [Mixer_1], properties {attenuationSettings=&Att_StealthSplash_Cloak_OnOff}}
 {id=Attenuation_2, children [Mixer_2], properties {attenuationSettings=&Att_StealthSplash_Cloak_OnOff}}
 {id=Branch_0, children [Attenuation_0, Branch_1, Branch_1], properties {boolParameterName=IsLocal}}
 {id=Branch_1, children [Attenuation_1, Attenuation_2, Attenuation_1], properties {boolParameterName=IsEnemy}}
 {id=Delay_0, children [WavePlayer_3], properties {delayMax=0.2, delayMin=0.2}}
 {id=Mixer_0, children [Mixer_3], properties {inputVolume [1]}}
 {id=Mixer_1, children [Mixer_3], properties {inputVolume [0.9]}}
 {id=Mixer_2, children [Mixer_3], properties {inputVolume [0.8]}}
 {id=Mixer_3, children [null,Modulator_0], properties {inputVolume [1,1]}}
 {id=Modulator_0, children [Delay_0], properties {pitchMax=1, pitchMin=1, volumeMax=1, volumeMin=1}}
 {id=WavePlayer_3, properties {soundWaveAssetPtr=&/SteathSplash_Cloak_Activate_01}}
]
properties {
 attenuationSettings=&Att_StealthSplash_Cloak_OnOff
 maxDistance=6300
 soundClassObject=&Fort_Gadgets_Items
 volumeMultiplier=0.8
}
