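# Sound cue for the StealthSplash item's cloak deactivation, in the gadget and item sound class.
# An IsLocal branch plays the local player's copy without an attenuation node and sends others to an IsEnemy branch.
# Enemy and non-enemy paths are attenuated, with enemies mixed slightly louder and an unset IsEnemy treated as enemy.
# All paths share one wave through a common modulator and mixer, whose first input is empty.
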SoundCue /Game/Athena/Sounds/Items/StealthSplash/StealthSplash_Cloak_Deactivate
root=Branch_0
nodes [
 {id=Attenuation_1, children [Mixer_1], properties {attenuationSettings=&/Att_StealthSplash_Cloak_OnOff}}
 {id=Attenuation_2, children [Mixer_2], properties {attenuationSettings=&Att_StealthSplash_Cloak_OnOff}}
 {id=Branch_0, children [Mixer_0,Branch_1,Branch_1], properties {boolParameterName=IsLocal}}
 {id=Branch_1, children [Attenuation_1, Attenuation_2, Attenuation_1], properties {boolParameterName=IsEnemy}}
 {id=Mixer_0, children [Mixer_3], properties {inputVolume [1]}}
 {id=Mixer_1, children [Mixer_3], properties {inputVolume [0.9]}}
 {id=Mixer_2, children [Mixer_3], properties {inputVolume [0.8]}}
 {id=Mixer_3, children [null,Modulator_0], properties {inputVolume [1,1]}}
 {id=Modulator_0, children [WavePlayer_1], properties {pitchMax=1, pitchMin=1, volumeMin=1}}
 {id=WavePlayer_1, properties {soundWaveAssetPtr=&/SteathSplash_Cloak_Deactivate_01}}
]
properties {
 attenuationSettings=&Att_StealthSplash_Cloak_OnOff
 maxDistance=6300
 soundClassObject=&Fort_Gadgets_Items
 volumeMultiplier=0.7
}
