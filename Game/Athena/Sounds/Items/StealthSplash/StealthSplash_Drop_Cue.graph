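# Sound cue for the drop sound of the StealthSplash item in Athena.
# A Random node picks between two drop wave variants at equal weight.
# It routes through the gadget items sound class with weapon-drop attenuation at reduced volume, and overrides concurrency to cap simultaneous plays, refusing new ones once the limit is reached.
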SoundCue /Game/Athena/Sounds/Items/StealthSplash/StealthSplash_Drop_Cue
root=Random_0
nodes [
 {id=Random_0, children [WavePlayer_0,WavePlayer_1], properties {bShouldExcludeFromBranchCulling=true, weights [0.8,0.8]}}
 {id=WavePlayer_0, properties {soundWaveAssetPtr=&/SteathSplash_Drop_01}}
 {id=WavePlayer_1, properties {soundWaveAssetPtr=&/SteathSplash_Drop_02}}
]
properties {
 attenuationSettings=&Fort_Weapon_Drop_Att
 bOverrideConcurrency=true
 maxDistance=3000
 soundClassObject=&Fort_Gadgets_Items
 volumeMultiplier=0.3
 concurrencyOverrides {maxCount=3, resolutionRule=PreventNew}
}
