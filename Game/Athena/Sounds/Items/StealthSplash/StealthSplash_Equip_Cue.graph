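# Sound cue that plays the player foley for equipping the StealthSplash item.
# A Random node picks with equal weight between two equip waves and an empty third slot, feeding a Modulator that varies pitch slightly downward at constant volume.
# It uses close-prop attenuation and overrides concurrency to prevent new instances once the limit is reached.
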SoundCue /Game/Athena/Sounds/Items/StealthSplash/StealthSplash_Equip_Cue
root=Modulator_2
nodes [
 {id=Modulator_2, children [Random_0], properties {pitchMax=1, volumeMax=1, volumeMin=1}}
 {id=Random_0, children [WavePlayer_0,WavePlayer_2,null], properties {bShouldExcludeFromBranchCulling=true, weights [1,1,1]}}
 {id=WavePlayer_0, properties {soundWaveAssetPtr=&/SteathSplash_Equip_01}}
 {id=WavePlayer_2, properties {soundWaveAssetPtr=&/SteathSplash_Equip_02}}
]
properties {
 attenuationSettings=/Game/Sounds/Attenuation/Fort_Prop_Close_Att
 bOverrideConcurrency=true
 maxDistance=1800
 soundClassObject=&Fort_Player_Foley
 volumeMultiplier=0.7
 concurrencyOverrides {resolutionRule=PreventNew}
}
