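# Sound cue for picking up the StealthSplash item, classed as a gadget item sound.
# A Random node picks between two equally weighted pickup waves and is excluded from branch culling.
# A Modulator above it varies pitch slightly downward while holding volume steady.
# The cue uses shared weapon-pickup attenuation, a reduced volume multiplier, and a concurrency override that prevents new instances at the limit.
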SoundCue /Game/Athena/Sounds/Items/StealthSplash/StealthSplash_Pickup_Cue
root=Modulator_2
nodes [
 {id=Modulator_2, children [Random_0], properties {pitchMax=1, volumeMax=1, volumeMin=1}}
 {id=Random_0, children [WavePlayer_1,WavePlayer_2], properties {bShouldExcludeFromBranchCulling=true, weights [1,1]}}
 {id=WavePlayer_1, properties {soundWaveAssetPtr=&/SteathSplash_PickUp_01}}
 {id=WavePlayer_2, properties {soundWaveAssetPtr=&/SteathSplash_PickUp_02}}
]
properties {
 attenuationSettings=&Fort_Weapon_Pickup_Att
 bOverrideConcurrency=true
 maxDistance=3500
 soundClassObject=&Fort_Gadgets_Items
 volumeMultiplier=0.8
 concurrencyOverrides {resolutionRule=PreventNew}
}
