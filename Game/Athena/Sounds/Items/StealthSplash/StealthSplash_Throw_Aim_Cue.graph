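# Sound cue for the aim sound played when throwing the Stealth Splash item in Athena.
# A Random node picks between two aim wave variants with equal weights, feeding a Modulator that slightly varies volume.
# It uses gadget-placement attenuation and the gadgets-items sound class at a reduced volume.
# An overridden concurrency rule blocks new instances rather than stopping playing ones.
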SoundCue /Game/Athena/Sounds/Items/StealthSplash/StealthSplash_Throw_Aim_Cue
root=Modulator_0
nodes [
 {id=Modulator_0, children [Random_0], properties {volumeMax=0.95}}
 {id=Random_0, children [WavePlayer_0,WavePlayer_2], properties {bShouldExcludeFromBranchCulling=true, weights [0.7,0.7]}}
 {id=WavePlayer_0, properties {soundWaveAssetPtr=&/SteathSplash_Aim_01}}
 {id=WavePlayer_2, properties {soundWaveAssetPtr=&/SteathSplash_Aim_02}}
]
properties {
 attenuationSettings=/Game/Sounds/Attenuation/Fort_GadgetPlace_Att
 bOverrideConcurrency=true
 maxDistance=2000
 soundClassObject=&Fort_Gadgets_Items
 volumeMultiplier=0.3
 concurrencyOverrides {resolutionRule=PreventNew}
}
