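# Sound cue for the throw sound of the Stealth Splash item, classed under gadget items.
# A Random node feeds a Modulator at default settings, choosing among two throw waves and an equally weighted empty third input.
# It borrows the explosive fuse projectile throw attenuation and overrides concurrency to prevent new instances once the limit is reached.
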SoundCue /Game/Athena/Sounds/Items/StealthSplash/StealthSplash_Throw_Cue
root=Modulator_2
nodes [
 {id=Modulator_2, children [Random_0]}
 {id=Random_0, children [WavePlayer_1,WavePlayer_2,null], properties {bShouldExcludeFromBranchCulling=true, weights [1,1,1]}}
 {id=WavePlayer_1, properties {soundWaveAssetPtr=&/SteathSplash_Throw_01}}
 {id=WavePlayer_2, properties {soundWaveAssetPtr=&/SteathSplash_Throw_02}}
]
properties {
 attenuationSettings=/Game/Sounds/Attenuation/Fort_Explosive_Fuse_Prj_Throw_Att
 bOverrideConcurrency=true
 maxDistance=13000
 soundClassObject=&Fort_Gadgets_Items
 volumeMultiplier=0.7
 concurrencyOverrides {resolutionRule=PreventNew}
}
